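# Conversation database for the NPC Trace, providing their greeting and top-level dialogue options.
# The entry point leads to one speech node with a default line, plus contextual lines for a first-ever conversation or specific player outfits, ranked by priority curves.
# That node branches through link nodes to shared NPC quest and basic services conversations and to four Trace-specific Telescope quest entries, Q01 to Q04.
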ConversationDatabase /NPCLibrary/NPCs/Trace/NPCLibrary_Conversation_Trace
entry_points [{tag=Conversation.Entry.Trace, destinations [#1]}]
nodes [
 ConversationEntryPointNode_7 {
  id=#1
  class=ConversationEntryPointNode
  outputs [#2]
  properties {entryTag {tagName=Conversation.Entry.Trace}}
 }
 Speech_C_0 {
  id=#2
  class=Speech_C
  outputs [
   #3
   #4
   #5
   #6
   #7
   #8
  ]
  properties {
   nodeName=""
   generalConfig {
    defaultMessage=Hey. How can the Luminaries be of service?
    contextualMessages [
     {
      message="You're that Hero of MEGA City, right? Nice to meet you."
      contextRequirements [{requirement {class=FortControllerRequirement_IsFirstEverConversationWithNPC}}]
      priority {curve {rowName=Tandem.Speech.Priority.FirstEverInteraction}}
     }
     {
      message=You know more about this place than anyone else on this Island. I'm curious how.
      contextRequirements [{requirement {class=FortControllerRequirement_HasCID, properties {allowedCIds [&Character_InnovatorSand]}}}]
      priority {curve {rowName=Tandem.Speech.Priority.OutfitReaction}}
     }
     {
      message="Thank you for believing me, that this project was worth the effort."
      contextRequirements [{requirement {class=FortControllerRequirement_HasCID, properties {allowedCIds [/BRCosmetics/Athena/Items/Cosmetics/Characters/Character_Nebula]}}}]
      priority {curve {rowName=Tandem.Speech.Priority.OutfitReaction}}
     }
     {
      message="You help me with this project, I'll help you track down your crew. Promise."
      contextRequirements [{requirement {class=FortControllerRequirement_HasCID, properties {allowedCIds [/BRCosmetics/Athena/Items/Cosmetics/Characters/Character_ScarletBionic]}}}]
      priority {curve {rowName=Tandem.Speech.Priority.OutfitReaction}}
     }
     {
      message="You really know this jungle like the back of your hand, huh?"
      contextRequirements [{requirement {class=FortControllerRequirement_HasCID, properties {allowedCIds [/BRCosmetics/Athena/Items/Cosmetics/Characters/Character_CinderMax]}}}]
      priority {curve {rowName=Tandem.Speech.Priority.OutfitReaction}}
     }
     {
      message="Stay out of our way, and we'll stay out of yours."
      contextRequirements [{requirement {class=FortControllerRequirement_HasCID, properties {allowedCIds [/BRCosmetics/Athena/Items/Cosmetics/Characters/Character_RippedHarvester]}}}]
      priority {curve {rowName=Tandem.Speech.Priority.OutfitReaction}}
     }
    ]
   }
  }
 }
 ConversationLinkNode_5 {
  id=#8
  class=ConversationLinkNode
  properties {remoteEntryTag {tagName=Conversation.Entry.Telescope.Q04.Trace}}
 }
 ConversationLinkNode_4 {
  id=#7
  class=ConversationLinkNode
  properties {remoteEntryTag {tagName=Conversation.Entry.Telescope.Q03.Trace}}
 }
 ConversationLinkNode_3 {
  id=#6
  class=ConversationLinkNode
  properties {remoteEntryTag {tagName=Conversation.Entry.Telescope.Q02.Trace}}
 }
 ConversationLinkNode_2 {
  id=#5
  class=ConversationLinkNode
  properties {remoteEntryTag {tagName=Conversation.Entry.Telescope.Q01.Trace}}
 }
 ConversationLinkNode_0 {id=#4, class=ConversationLinkNode, properties {remoteEntryTag {tagName=Conversation.Entry.BasicServices}}}
 ConversationLinkNode_1 {id=#3, class=ConversationLinkNode, properties {remoteEntryTag {tagName=Conversation.Entry.NPCLibrary.Quests}}}
]
